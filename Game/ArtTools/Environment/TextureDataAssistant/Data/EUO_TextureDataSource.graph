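# Editor utility object for the Texture Data Assistant that gathers one set of source textures and an optional override material.
# Added textures are classified by LOD group and compression settings into Diffuse, Normal, Spec, Displacement or Invalid slots.
# SetRootName derives a TD_ display name, stripping any T_ prefix.
# FinalizeSourceData stores the first valid asset's folder as AssetPath; array appends and full-path building are left unwired.
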EditorUtilityBlueprint /Game/ArtTools/Environment/TextureDataAssistant/Data/EUO_TextureDataSource
graphs [
 AddAssetToSource {nodes [
   n1 = AddAssetToSource!() -> (then, Asset:Object Reference) {is=title K2Node_FunctionEntry}
   > n2 = Cast To Texture2D(Object=@n1)
   > AddTextureToSource!(execute, self, Texture:Texture 2D Object Reference=@n2) -> (then)
   n3 = Cast To MaterialInterface!(execute=@n2.CastFailed, Object=@n1) -> (then, CastFailed, AsMaterial Interface:Material Interface Object Reference) {is=title K2Node_DynamicCast}
   > AddMaterialToSource!(execute, self, Material:Material Interface Object Reference=@n3) -> (then)
  ]}
 AddMaterialToSource {nodes [
   n1 = AddMaterialToSource!() -> (then, Material:Material Interface Object Reference) {is=title K2Node_FunctionEntry}
   > Set Override Material(Override Material:Material Interface Object Reference=@n1) -> (Output_Get:Material Interface Object Reference)
   Add(
    TargetArray:Array of Material Interface Object References=Get AR_Materials() -> (AR_Materials:Array of Material Interface Object References)
    NewItem:"Material Interface Object Reference (by ref)"
   ) {is=title K2Node_CallArrayFunction}
  ]}
 AddTextureToSource {nodes [
   n1 = AddTextureToSource!() -> (then, Texture:Texture 2D Object Reference) {is=title K2Node_FunctionEntry}
   > n2 = Set L_Texture(L_Texture:Texture 2D Object Reference=@n1) -> (Output_Get:Texture 2D Object Reference)
   > n3 = GetTextureType!(execute, self, Texture:Texture Object Reference=@n2) -> (then, TextureType:ENUM_TDA_TextureType Enum=NewEnumerator0)
   > n4 = Switch on ENUM_TDA_TextureType!(execute, Selection:ENUM_TDA_TextureType Enum=@n3) -> (NewEnumerator0:Exec, NewEnumerator1:Exec, NewEnumerator2:Exec, NewEnumerator3:Exec, NewEnumerator4:Exec) {is=title K2Node_SwitchEnum}
   n5 = Get L_Texture() -> (L_Texture:Texture 2D Object Reference)
   > Set Diffuse(Diffuse:Texture 2D Object Reference=@n5) -> (Output_Get:Texture 2D Object Reference)
   Set Normal(execute=@n4.NewEnumerator1, Normal:Texture 2D Object Reference=@n5) -> (Output_Get:Texture 2D Object Reference)
   Set Spec(execute=@n4.NewEnumerator2, Spec:Texture 2D Object Reference=@n5) -> (Output_Get:Texture 2D Object Reference)
   Set Displacement(execute=@n4.NewEnumerator3, Displacement:Texture 2D Object Reference=@n5) -> (Output_Get:Texture 2D Object Reference)
   Set Invalid(execute=@n4.NewEnumerator4, Invalid:Texture 2D Object Reference=@n5) -> (Output_Get:Texture 2D Object Reference)
   n6 = Switch on ENUM_TDA_TextureType(Selection=NewEnumerator0)
   n7 = Get L_Texture()
   > Add(
    TargetArray:Array of Texture 2D Object References=Get AR_DiffuseTextures() -> (AR_DiffuseTextures:Array of Texture 2D Object References)
    NewItem:"Texture 2D Object Reference (by ref)"=@n7
   )
   Add(execute=@n6.NewEnumerator1, TargetArray=Get AR_NormalTextures() -> (AR_NormalTextures:Array of Texture 2D Object References), NewItem=@n7)
   Add(execute=@n6.NewEnumerator2, TargetArray=Get AR_SpecTextures() -> (AR_SpecTextures:Array of Texture 2D Object References), NewItem=@n7)
   Add(
    execute=@n6.NewEnumerator3
    TargetArray=Get AR_DisplaceTextures() -> (AR_DisplaceTextures:Array of Texture 2D Object References)
    NewItem=@n7
   )
  ]}
 EventGraph {links [], nodes []}
 FinalizeSourceData {nodes [
   FinalizeSourceData() {is=title K2Node_FunctionEntry}
   > n1 = TEMP_GetValidAssetFromSource!(execute, self) -> (then, SourcePath:String, WasMaterial:bool)
   > ValidateAssetPath!(execute, self, SourcePath:String=@n1) -> (then)
  ]}
 GetTextureType {nodes [
   n1 = GetTextureType!() -> (then, Texture:Texture Object Reference) {is=title K2Node_FunctionEntry}
   > n2 = IdentifyTextureType!(execute, self, Texture:Texture Object Reference=@n1) -> (then, TextureGroup:ENUM_TDA_TextureType Enum=NewEnumerator0)
   > Return Node!(execute, TextureType:ENUM_TDA_TextureType Enum=@n2)
  ]}
 IdentifyTextureType {nodes [
   n1 = IdentifyTextureType!() -> (then, Texture:Texture Object Reference) {is=title K2Node_FunctionEntry}
   n2 = Get LODGroup!(self:Texture Object Reference=@n1) -> (LODGroup:TextureGroup Enum)
   > n3 = Branch(Condition="Equal (Enum)"(A:TextureGroup Enum=@n2, B:TextureGroup Enum=TEXTUREGROUP_World))
   n4 = Get CompressionSettings(self=@n1)
   > n5 = Branch(Condition="Equal (Enum)"(A:TextureCompressionSettings Enum=@n4, B:TextureCompressionSettings Enum=TC_Default))
   > n6 = Set L_TextureGroup(L_TextureGroup:ENUM_TDA_TextureType Enum=NewEnumerator0) -> (Output_Get:ENUM_TDA_TextureType Enum=NewEnumerator0)
   Return Node!(
    execute=[@n8.then, @n6.then, @n11.then, @n12.then, @n9.then]
    TextureGroup:ENUM_TDA_TextureType Enum=Get L_TextureGroup() -> (L_TextureGroup:ENUM_TDA_TextureType Enum=NewEnumerator0)
   )
   n7 = Branch(execute=@n5.else, Condition="Equal (Enum)"(A=@n4, B=TC_Grayscale))
   > n8 = Set L_TextureGroup(L_TextureGroup=NewEnumerator3) -> (Output_Get=NewEnumerator0)
   n9 = Set L_TextureGroup(execute=@n7.else, L_TextureGroup=NewEnumerator4) -> (Output_Get=NewEnumerator0)
   n10 = Branch(execute=@n3.else, Condition="Equal (Enum)"(A:TextureGroup Enum=@n2, B:TextureGroup Enum=TEXTUREGROUP_WorldNormalMap))
   > Branch(Condition="Equal (Enum)"(A:TextureCompressionSettings Enum=@n4, B:TextureCompressionSettings Enum=TC_Normalmap))
   > n11 = Set L_TextureGroup(L_TextureGroup=NewEnumerator1) -> (Output_Get=NewEnumerator0)
   Branch(execute=@n10.else, Condition="Equal (Enum)"(A:TextureGroup Enum=@n2, B:TextureGroup Enum=TEXTUREGROUP_WorldSpecular))
   > Branch(Condition="Equal (Enum)"(A:TextureCompressionSettings Enum=@n4, B:TextureCompressionSettings Enum=TC_Masks))
   > n12 = Set L_TextureGroup(L_TextureGroup=NewEnumerator2) -> (Output_Get=NewEnumerator0)
  ]}
 SetRootName {nodes [
   n1 = SetRootName!() -> (then, RootName:String) {is=title K2Node_FunctionEntry}
   > n2 = Set RootName(RootName:String=@n1) -> (Output_Get:String)
   > n3 = Set DisplayName(DisplayName:String=@n2) -> (Output_Get:String)
   > n4 = Sequence()
   > Branch(Condition=StartsWith(SourceString=@n3, InPrefix=T_))
   > Set DisplayName(DisplayName=RightChop!(SourceString:String=@n3, Count:int=2) -> (ReturnValue:String))
   Set DisplayName(execute=@n4.then_1, DisplayName=Append(A=TD_, B=Get DisplayName() -> (DisplayName:String)))
  ]}
 TEMP_GetValidAssetFromSource {nodes [
   TEMP_GetValidAssetFromSource() {is=title K2Node_FunctionEntry}
   > n1 = For Each Loop(Array:Array of Object References=Make Array!(
     "[0]":Object Reference=Get Diffuse() -> (Diffuse:Texture 2D Object Reference)
     "[1]":Object Reference=Get Normal() -> (Normal:Texture 2D Object Reference)
     "[2]":Object Reference=Get Spec() -> (Spec:Texture 2D Object Reference)
     "[3]":Object Reference=Get Displacement() -> (Displacement:Texture 2D Object Reference)
     "[4]":Object Reference=Get Override Material() -> (Override Material:Material Interface Object Reference)
    ) -> (Array:Array of Object References)) -> (Array Element:Object Reference)
   > n2 = Set L_Asset(L_Asset:Object Reference=@n1) -> (Output_Get:Object Reference)
   > Is Valid(InputObject=@n2)
   n3 = Get L_Asset() -> (L_Asset:Object Reference)
   > n4 = Set SourcePath(SourcePath:String="To String (Name)"(InName=CreateAssetData!(InAsset=@n3, bAllowBlueprintClass=false) -> (
      ReturnValue_PackageName:Name
      ReturnValue_PackagePath:Name
      ReturnValue_AssetName:Name
      ReturnValue_AssetClassPath:Top Level Asset Path Structure
     ).ReturnValue_PackagePath)) -> (Output_Get:String)
   > Return Node!(
    execute
    SourcePath:String=@n4
    WasMaterial:bool=Cast To MaterialInterface!(Object=Get L_Asset()) -> (AsMaterial Interface, bSuccess:bool=~).bSuccess
   )
  ]}
 ValidateAssetPath {nodes [
   n1 = ValidateAssetPath!() -> (then, SourcePath:String) {is=title K2Node_FunctionEntry}
   > Set AssetPath(AssetPath:String=@n1) -> (Output_Get:String)
   ParseIntoArray(Delimiter=" ")
   n2 = Get DisplayName()
   Append!(A, B=/, C:String=TextureData, D:String=/, E:String=@n2, F:String=., G:String=@n2) -> (ReturnValue)
  ]}
]
variables [
 AR_DiffuseTextures {type="array<Texture2D>", default []}
 AR_DisplaceTextures {type="array<Texture2D>", default []}
 AR_Materials {type="array<MaterialInterface>", default []}
 AR_NormalTextures {type="array<Texture2D>", default []}
 AR_SpecTextures {type="array<Texture2D>", default []}
 AssetPath {type=string, default=""}
 Condition {type=boolean, default=true}
 Diffuse {type=Texture2D}
 Displacement {type=Texture2D}
 DisplayName {type=string, default=""}
 Invalid {type=Texture2D}
 NewTD {type=Object, default=null}
 Normal {type=Texture2D}
 Override Material {type=MaterialInterface, default=null}
 RootName {type=string, default=""}
 SourcePath {type=string, default=""}
 Spec {type=Texture2D}
]
