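# Editor utility list-entry widget for one texture source in the Texture Data Assistant art tool.
# When assigned a list item, it casts it to EUO_TextureDataSource and stores it as CurrentData.
# It shows the source's RootName in a label and binds the object to a scrollbar-less Details View below, filtered by category, so its properties display inline.
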EditorUtilityWidgetBlueprint /Game/ArtTools/Environment/TextureDataAssistant/Widgets/EUW_TextureSourceItem
graphs [EventGraph {nodes [
   n1 = Event OnListItemObjectSet()
   > n2 = Cast To EUO_TextureDataSource!(execute, Object=@n1.ListItemObject) -> (then, CastFailed, AsEUO Texture Data Source:EUO Texture Data Source Object Reference) {is=title K2Node_DynamicCast}
   > Set CurrentData(CurrentData:EUO Texture Data Source Object Reference=@n2) -> (Output_Get:EUO Texture Data Source Object Reference)
   > "SetText (Text)"(
    self=Get TXT_Label() -> (TXT_Label:Text Object Reference)
    InText="To Text (String)"(InString=Get RootName!(self:EUO Texture Data Source Object Reference=@n2) -> (RootName:String))
   )
   > SetObject(
    self=Get DetailsView_104() -> (DetailsView_104:Details View Object Reference)
    NewObject=Get CurrentData() -> (CurrentData:EUO Texture Data Source Object Reference)
   )
   Event PreConstruct()
   Event Construct()
   Event Tick()
  ]}]
variables [CurrentData {type=EUO_TextureDataSource_C, default=null}]
widgets {
 info {rootWidgetClass=&StackBox, widgetCount=3}
 widgets [
  StackBox_0 {widgetClassPath=&StackBox, widget=StackBox_0, properties {orientation=Orient_Vertical}}
  TXT_Label {
   widgetClassPath=&TextBlock
   bIsVariable=true
   parent=StackBox_0
   widget=TXT_Label
   properties {text=Label}
   slot {class=StackBoxSlot}
  }
  DetailsView_104 {
   widgetClassPath=&DetailsView
   parent=StackBox_0
   widget=DetailsView_104
   properties {bShowScrollBar=false, categoriesToShow [nocategory]}
   slot {class=StackBoxSlot}
  }
 ]
}
