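# Editor utility widget that drafts texture data from texture assets; its Create Assets function is an empty stub.
# Generate groups assets by name up to the last underscore into EUO_TextureDataSource objects, then finalizes and lists them.
# GetSourceAssets returns a local override array, leaving GetSelectedAssets unconnected despite the button's "From Selection" label.
# A side button, shown only to one hard-coded username, opens the widget's Blueprint through Python.
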EditorUtilityWidgetBlueprint /Game/ArtTools/Environment/TextureDataAssistant/EUW_TextureDataAssistant
graphs [
 CreateTextureData {nodes [CreateTextureData() {is=title K2Node_FunctionEntry}]}
 EventGraph {nodes [
   "On Clicked (BTN_Generate)"() {is=title K2Node_ComponentBoundEvent}
   > GenerateTextureDataSources()
   > PopulateSourceList()
   "On Clicked (BTN_Create)"() {is=title K2Node_ComponentBoundEvent}
   > CreateTextureData() {is=title K2Node_CallFunction}
   "On Clicked (BTN_EditWidget)"!() -> (OutputDelegate, then, widget:User Widget Object Reference, event:Pointer Event Structure) {is=title K2Node_ComponentBoundEvent}
   > n1 = GetBlueprintAsset(Object=&self)
   > Execute Python Script!(
    execute
    PythonScript="import unreal\r\n\r\nmy_widget = widget_object\r\nasset_editor_subsystem = unreal.get_editor_subsystem(unreal.AssetEditorSubsystem)\r\nasset_editor_subsystem.open_editor_for_assets([my_widget])"
    widget_object:Blueprint Object Reference=@n1
   ) -> (then, ReturnValue) {inputs [widget_object]}
   Event Construct()
   > n2 = Branch(Condition="Equal Exactly (String)"(A=GetPlatformUserName!() -> (ReturnValue:String), B=brianfasten))
   n3 = Get BTN_EditWidget() -> (BTN_EditWidget:TA Generic 01 Button Object Reference)
   > SetVisibility(self=@n3, InVisibility=Visible)
   SetVisibility(execute=@n2.else, self=@n3, InVisibility=Hidden)
   DuplicateLoadedAsset!(
    execute
    SourceAsset:Object Reference=/Game/Environments/Outsource/Templates/TD_Template
    DestinationAssetPath:String=/Game/Environments/Outsource/Templates/Test
   ) -> (then, ReturnValue:Object Reference)
   Event PreConstruct()
   Event Tick()
  ]}
 GenerateTextureDataSources {nodes [
   GenerateTextureDataSources() {is=title K2Node_FunctionEntry}
   > n1 = GetSourceAssets!(execute, self) -> (then, AR_Assets:Array of Object References)
   > n2 = For Each Loop(Array:Array of Object References=@n1) -> (Array Element:Object Reference)
   Clear(
    execute=@n2.Completed
    TargetArray:Array of EUO Texture Data Source Object References=Get AR_SourceObjects() -> (AR_SourceObjects:Array of EUO Texture Data Source Object References)
   )
   > n3 = Values!(
    execute
    TargetMap:Map of Strings to EUO Texture Data Source Object References=Get L_RootSourceMap() -> (L_RootSourceMap:Map of Strings to EUO Texture Data Source Object References)
   ) -> (then, Values:Array of EUO Texture Data Source Object References)
   > Set AR_SourceObjects(AR_SourceObjects:Array of EUO Texture Data Source Object References=@n3) -> (Output_Get:Array of EUO Texture Data Source Object References)
   n4 = Set L_Asset(execute=@n2.LoopBody, L_Asset:Object Reference=@n2) -> (Output_Get:Object Reference)
   > n5 = GetNameRootOfAsset!(execute, self, Object:Object Reference=@n4) -> (then, RootName:String, Valid:bool)
   > Set L_RootName(L_RootName:String=@n5) -> (Output_Get:String)
   > Branch(Condition=@n5.Valid)
   > n6 = Find!(
    execute:Exec
    TargetMap:Map of Strings to EUO Texture Data Source Object References=Get L_RootSourceMap()
    Key=Get L_RootName() -> (L_RootName:String)
   ) -> (then:Exec, Value:EUO Texture Data Source Object Reference, ReturnValue)
   > n7 = Branch(Condition=@n6.ReturnValue)
   > n8 = Set L_Source(L_Source:EUO Texture Data Source Object Reference=@n6) -> (Output_Get:EUO Texture Data Source Object Reference)
   AddAssetToSource!(
    execute=[@n11.then, @n8.then]
    self:EUO Texture Data Source Object Reference=Get L_Source() -> (L_Source:EUO Texture Data Source Object Reference)
    Asset:Object Reference=Get L_Asset() -> (L_Asset:Object Reference)
   ) -> (then)
   n9 = Construct Object from Class(execute=@n7.else, Class=/Game/ArtTools/Environment/TextureDataAssistant/Data/EUO_TextureDataSource.EUO_TextureDataSource_C) -> (ReturnValue:EUO Texture Data Source Object Reference)
   > n10 = Set L_Source(L_Source=@n9)
   > SetRootName!(
    execute
    self:EUO Texture Data Source Object Reference=@n10
    RootName:String=Replace(SourceString=Get L_RootName(), From=T_, SearchCase=CaseSensitive)
   ) -> (then)
   > n11 = Add!(
    execute
    TargetMap:Map of Strings to EUO Texture Data Source Object References=Get L_RootSourceMap()
    Key:"String (by ref)"=Get L_RootName()
    Value:"EUO Texture Data Source Object Reference (by ref)"=Get L_Source()
   ) -> (then) {is=title K2Node_CallFunction}
   ValidateAssetPath!(execute, self:EUO Texture Data Source Object Reference=Get L_Source(), SourcePath:String) -> (then)
  ]}
 GetNameRootOfAsset {nodes [
   n1 = GetNameRootOfAsset!() -> (then, Object:Object Reference) {is=title K2Node_FunctionEntry}
   > n2 = Split!(execute:Exec, SourceString=GetDisplayName(Object=@n1), InStr=_, SearchCase, SearchDir=FromEnd) -> (then:Exec, LeftS, RightS, ReturnValue)
   > Return Node!(execute, RootName:String=@n2, Valid:bool=@n2.ReturnValue)
  ]}
 GetSourceAssets {nodes [
   GetSourceAssets() {is=title K2Node_FunctionEntry}
   > Return Node!(execute, AR_Assets:Array of Object References=Get L_AR_TempAssetOverride() -> (L_AR_TempAssetOverride:Array of Object References))
   GetSelectedAssets()
  ]}
 PopulateInvalidList {nodes [
   ClearListItems()
   > n1 = For Each Loop(Array:Array of EUO Texture Data Source Object References) -> (Array Element:EUO Texture Data Source Object Reference)
   > AddItem!(
    execute
    self:List View Object Reference=Reroute Node(InputPin:List View Object Reference) -> (OutputPin:List View Object Reference)
    Item:Object Reference=@n1
   ) -> (then)
   PopulateInvalidList() {is=title K2Node_FunctionEntry}
  ]}
 PopulateSourceList {nodes [
   PopulateSourceList() {is=title K2Node_FunctionEntry}
   n1 = Get LIST_TextureDataObjects() -> (LIST_TextureDataObjects:Common List View Object Reference)
   > ClearListItems(self=@n1)
   > n2 = For Each Loop(Array=Get AR_SourceObjects())
   > FinalizeSourceData(self:EUO Texture Data Source Object Reference=@n2)
   > AddItem(self=@n1, Item=@n2)
  ]}
]
variables [AR_SourceObjects {type="array<EUO_TextureDataSource_C>", default []}]
widgets {
 info {rootWidgetClass=&HorizontalBox, widgetCount=8}
 widgets [
  HorizontalBox_0 {widgetClassPath=&HorizontalBox, widget=HorizontalBox_0}
  StackBox_0 {
   widgetClassPath=&StackBox
   parent=HorizontalBox_0
   widget=StackBox_0
   properties {orientation=Orient_Vertical}
   slot {class=HorizontalBoxSlot, properties {size {sizeRule=Fill, value=1}}}
  }
  BTN_Generate {widgetClassPath=&EditorUtilityButton, parent=StackBox_0, widget=BTN_Generate, slot {class=StackBoxSlot}}
  TextBlock_118 {
   widgetClassPath=&TextBlock
   parent=BTN_Generate
   widget=TextBlock_118
   properties {
    text=Generate Texture Data From Selection
    font {
     bForceMonospaced=false
     bMaterialIsStencil=false
     fontMaterial=null
     fontObject=&Roboto
     letterSpacing=0
     monospacedWidth=1
     size=12
     skewAmount=0
     typefaceFontName=Bold
     outlineSettings {
      bApplyOutlineToDropShadows=false
      bMiteredCorners=false
      bSeparateFillAlpha=false
      outlineMaterial=null
      outlineSize=0
      outlineColor=<rgba 0 0 0 1>
     }
    }
   }
   slot {class=ButtonSlot}
  }
  LIST_TextureDataObjects {
   widgetClassPath=&CommonListView
   bIsVariable=true
   parent=StackBox_0
   widget=LIST_TextureDataObjects
   properties {
    entryWidgetClass=/Game/ArtTools/Environment/TextureDataAssistant/Widgets/EUW_TextureSourceItem.EUW_TextureSourceItem_C
    scrollBarStyle {
     thickness=8
     draggedThumbImage {
      bIsDynamicallyLoaded=false
      drawAs=RoundedBox
      imageType=NoImage
      mirroring=NoMirror
      resourceName=None
      resourceObject=null
      tiling=NoTile
      imageSize=<0 0>
      margin=<ltrb 0 0 0 0>
      outlineSettings {
       bUseBrushTransparency=false
       roundingType=FixedRadius
       width=0
       color {colorUseRule=UseColor_Specified, specifiedColor=<rgba 0 0 0 0>}
       cornerRadii=<4 4 4 4>
      }
      tintColor {colorUseRule=UseColor_Specified, specifiedColor=<rgba 0.2159 0.2159 0.2159 1>}
      uVRegion {bIsValid=false, max=<0 0>, min=<0 0>}
     }
     horizontalBackgroundImage {
      bIsDynamicallyLoaded=false
      drawAs=NoDrawType
      imageType=NoImage
      mirroring=NoMirror
      resourceName=None
      resourceObject=null
      tiling=NoTile
      imageSize=<0 0>
      margin=<ltrb 0 0 0 0>
      outlineSettings {
       bUseBrushTransparency=false
       roundingType=HalfHeightRadius
       width=0
       color {colorUseRule=UseColor_Specified, specifiedColor=<rgba 0 0 0 0>}
       cornerRadii=<0 0 0 0>
      }
      tintColor {colorUseRule=UseColor_Specified, specifiedColor=<rgba 1 1 1 1>}
      uVRegion {bIsValid=false, max=<0 0>, min=<0 0>}
     }
     horizontalBottomSlotImage {
      bIsDynamicallyLoaded=false
      drawAs=NoDrawType
      imageType=NoImage
      mirroring=NoMirror
      resourceName=None
      resourceObject=null
      tiling=NoTile
      imageSize=<0 0>
      margin=<ltrb 0 0 0 0>
      outlineSettings {
       bUseBrushTransparency=false
       roundingType=HalfHeightRadius
       width=0
       color {colorUseRule=UseColor_Specified, specifiedColor=<rgba 0 0 0 0>}
       cornerRadii=<0 0 0 0>
      }
      tintColor {colorUseRule=UseColor_Specified, specifiedColor=<rgba 1 1 1 1>}
      uVRegion {bIsValid=false, max=<0 0>, min=<0 0>}
     }
     horizontalTopSlotImage {
      bIsDynamicallyLoaded=false
      drawAs=NoDrawType
      imageType=NoImage
      mirroring=NoMirror
      resourceName=None
      resourceObject=null
      tiling=NoTile
      imageSize=<0 0>
      margin=<ltrb 0 0 0 0>
      outlineSettings {
       bUseBrushTransparency=false
       roundingType=HalfHeightRadius
       width=0
       color {colorUseRule=UseColor_Specified, specifiedColor=<rgba 0 0 0 0>}
       cornerRadii=<0 0 0 0>
      }
      tintColor {colorUseRule=UseColor_Specified, specifiedColor=<rgba 1 1 1 1>}
      uVRegion {bIsValid=false, max=<0 0>, min=<0 0>}
     }
     hoveredThumbImage {
      bIsDynamicallyLoaded=false
      drawAs=RoundedBox
      imageType=NoImage
      mirroring=NoMirror
      resourceName=None
      resourceObject=null
      tiling=NoTile
      imageSize=<0 0>
      margin=<ltrb 0 0 0 0>
      outlineSettings {
       bUseBrushTransparency=false
       roundingType=FixedRadius
       width=0
       color {colorUseRule=UseColor_Specified, specifiedColor=<rgba 0 0 0 0>}
       cornerRadii=<4 4 4 4>
      }
      tintColor {colorUseRule=UseColor_Specified, specifiedColor=<rgba 0.2159 0.2159 0.2159 1>}
      uVRegion {bIsValid=false, max=<0 0>, min=<0 0>}
     }
     normalThumbImage {
      bIsDynamicallyLoaded=false
      drawAs=RoundedBox
      imageType=NoImage
      mirroring=NoMirror
      resourceName=None
      resourceObject=null
      tiling=NoTile
      imageSize=<0 0>
      margin=<ltrb 0 0 0 0>
      outlineSettings {
       bUseBrushTransparency=false
       roundingType=FixedRadius
       width=0
       color {colorUseRule=UseColor_Specified, specifiedColor=<rgba 0 0 0 0>}
       cornerRadii=<4 4 4 4>
      }
      tintColor {colorUseRule=UseColor_Specified, specifiedColor=<rgba 0.0953 0.0953 0.0953 1>}
      uVRegion {bIsValid=false, max=<0 0>, min=<0 0>}
     }
     verticalBackgroundImage {
      bIsDynamicallyLoaded=false
      drawAs=NoDrawType
      imageType=NoImage
      mirroring=NoMirror
      resourceName=None
      resourceObject=null
      tiling=NoTile
      imageSize=<0 0>
      margin=<ltrb 0 0 0 0>
      outlineSettings {
       bUseBrushTransparency=false
       roundingType=HalfHeightRadius
       width=0
       color {colorUseRule=UseColor_Specified, specifiedColor=<rgba 0 0 0 0>}
       cornerRadii=<0 0 0 0>
      }
      tintColor {colorUseRule=UseColor_Specified, specifiedColor=<rgba 1 1 1 1>}
      uVRegion {bIsValid=false, max=<0 0>, min=<0 0>}
     }
     verticalBottomSlotImage {
      bIsDynamicallyLoaded=false
      drawAs=NoDrawType
      imageType=NoImage
      mirroring=NoMirror
      resourceName=None
      resourceObject=null
      tiling=NoTile
      imageSize=<0 0>
      margin=<ltrb 0 0 0 0>
      outlineSettings {
       bUseBrushTransparency=false
       roundingType=HalfHeightRadius
       width=0
       color {colorUseRule=UseColor_Specified, specifiedColor=<rgba 0 0 0 0>}
       cornerRadii=<0 0 0 0>
      }
      tintColor {colorUseRule=UseColor_Specified, specifiedColor=<rgba 1 1 1 1>}
      uVRegion {bIsValid=false, max=<0 0>, min=<0 0>}
     }
     verticalTopSlotImage {
      bIsDynamicallyLoaded=false
      drawAs=NoDrawType
      imageType=NoImage
      mirroring=NoMirror
      resourceName=None
      resourceObject=null
      tiling=NoTile
      imageSize=<0 0>
      margin=<ltrb 0 0 0 0>
      outlineSettings {
       bUseBrushTransparency=false
       roundingType=HalfHeightRadius
       width=0
       color {colorUseRule=UseColor_Specified, specifiedColor=<rgba 0 0 0 0>}
       cornerRadii=<0 0 0 0>
      }
      tintColor {colorUseRule=UseColor_Specified, specifiedColor=<rgba 1 1 1 1>}
      uVRegion {bIsValid=false, max=<0 0>, min=<0 0>}
     }
    }
   }
   slot {class=StackBoxSlot, properties {size {sizeRule=Fill, value=1}}}
  }
  BTN_Create {widgetClassPath=&EditorUtilityButton, parent=StackBox_0, widget=BTN_Create, slot {class=StackBoxSlot}}
  TextBlock {
   widgetClassPath=&TextBlock
   parent=BTN_Create
   widget=TextBlock
   properties {
    text=Create Assets
    font {
     bForceMonospaced=false
     bMaterialIsStencil=false
     fontMaterial=null
     fontObject=&Roboto
     letterSpacing=0
     monospacedWidth=1
     size=12
     skewAmount=0
     typefaceFontName=Bold
     outlineSettings {
      bApplyOutlineToDropShadows=false
      bMiteredCorners=false
      bSeparateFillAlpha=false
      outlineMaterial=null
      outlineSize=0
      outlineColor=<rgba 0 0 0 1>
     }
    }
   }
   slot {class=ButtonSlot}
  }
  BTN_EditWidget {
   widgetClassPath=&TA_Generic01_Button
   parent=HorizontalBox_0
   widget=BTN_EditWidget
   properties {label=">>", buttonColor_Default=<rgba 0.0469 0.0469 0.0469 1>, sizeOverride=<20 0>}
   slot {class=HorizontalBoxSlot, properties {padding=<ltrb 2 0 2 0>}}
  }
 ]
}
